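# Physics asset for the M_MED_York male player body, setting up secondary motion for its buckles, straps, zippers and left-side tubes.
# Each dangling piece is a capsule or chain, angle-limited against a tiny MSTR anchor body.
# Anchors connect to the pelvis with free-rotation constraints, except the lower front tube's, which connects to thigh_l.
# Front and back tube chains join at their tips with free-rotation constraints, forming two hanging loops.
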PhysicsAsset /BRCosmetics/Characters/Player/Male/Medium/Bodies/M_MED_York/Meshes/M_MED_York_Physics
bodies {
 dyn_buckle_01_l [{shapeType=Capsule, length=1.7665, radius=0.6376, center=<1.3702 -0.0383 0.0703>, rotation=<rot 0 88.6003 -87.2529>}]
 dyn_buckle_01_l_MSTR [{shapeType=Capsule, length=0.01, radius=0.01}]
 dyn_buckle_01_r [{shapeType=Capsule, length=1.7665, radius=0.6376, center=<1.3702 -0.0383 0.0703>, rotation=<rot 0 88.6003 -87.2529>}]
 dyn_buckle_01_r_MSTR [{shapeType=Capsule, length=0.01, radius=0.01}]
 dyn_strap_01_l [{shapeType=Capsule, length=4.2327, radius=1.5437, center=<4.9953 0.033 0.1771>, rotation=<rot 0 91.8176 -84.2551>}]
 dyn_strap_01_l_MSTR [{shapeType=Capsule, length=0.01, radius=0.01}]
 dyn_strap_01_r [{shapeType=Capsule, length=4.2326, radius=1.5438, center=<4.9948 -0.0329 0.1771>, rotation=<rot 0 88.1821 -84.2556>}]
 dyn_strap_01_r_MSTR [{shapeType=Capsule, length=0.01, radius=0.01}]
 dyn_tube_bk_01_l [{shapeType=Capsule, length=8.3179, radius=2.2983, center=<5.4723 0 -0.0004>, rotation=<rot 90 -0.0198 -0.0199>}]
 dyn_tube_bk_01_l_MSTR [{shapeType=Capsule, length=0.01, radius=0.01}]
 dyn_tube_bk_02_l [{shapeType=Capsule, length=8.3179, radius=2.2983, center=<5.4723 0 -0.0004>, rotation=<rot 90 -0.0198 -0.0199>}]
 dyn_tube_bk_03_l [{shapeType=Capsule, length=8.3179, radius=2.2983, center=<5.4723 0 -0.0004>, rotation=<rot 90 -0.0198 -0.0199>}]
 dyn_tube_fr_01_l [{shapeType=Capsule, length=8.3179, radius=2.2983, center=<5.4723 0 -0.0004>, rotation=<rot 90 -0.0198 -0.0199>}]
 dyn_tube_fr_01_l_MSTR [{shapeType=Capsule, length=0.01, radius=0.01}]
 dyn_tube_fr_02_l [{shapeType=Capsule, length=8.3179, radius=2.2983, center=<5.4723 0 -0.0004>, rotation=<rot 90 -0.0198 -0.0199>}]
 dyn_tubeupper_bk_01_l [{shapeType=Capsule, length=5.5189, radius=1.3153, center=<5.4723 0 -0.0004>, rotation=<rot 90 -0.0198 -0.0199>}]
 dyn_tubeupper_bk_01_l_MSTR [{shapeType=Capsule, length=0.01, radius=0.01}]
 dyn_tubeupper_bk_02_l [{shapeType=Capsule, length=5.5189, radius=1.3153, center=<5.4723 0 -0.0004>, rotation=<rot 90 -0.0198 -0.0199>}]
 dyn_tubeupper_fr_01_l [{shapeType=Capsule, length=5.5189, radius=1.3153, center=<5.4723 0 -0.0004>, rotation=<rot 90 -0.0198 -0.0199>}]
 dyn_tubeupper_fr_01_l_MSTR [{shapeType=Capsule, length=0.01, radius=0.01}]
 dyn_tubeupper_fr_02_l [{shapeType=Capsule, length=5.5189, radius=1.3153, center=<5.4723 0 -0.0004>, rotation=<rot 90 -0.0198 -0.0199>}]
 dyn_zipper_01 [{shapeType=Capsule, length=2.8248, radius=0.7786, center=<2.4642 -0.0145 -0.1107>, rotation=<rot 0 -90.6158 -88.0477>}]
 dyn_zipper_01_MSTR [{shapeType=Capsule, length=0.01, radius=0.01}]
 dyn_zipper_01_l [{shapeType=Capsule, length=1.7665, radius=0.6376, center=<1.3702 -0.0383 0.0703>, rotation=<rot 0 88.6003 -87.2529>}]
 dyn_zipper_01_l_MSTR [{shapeType=Capsule, length=0.01, radius=0.01}]
 dyn_zipper_01_r [{shapeType=Capsule, length=1.7692, radius=0.6406, center=<1.366 0.0962 0.097>, rotation=<rot 0 91.3851 -87.2216>}]
 dyn_zipper_01_r_MSTR [{shapeType=Capsule, length=0.01, radius=0.01}]
 pelvis [{shapeType=Capsule, length=11.1305, radius=9.6752, center=<-2.7401 0 0>}]
 thigh_l [{shapeType=Capsule, length=34.9462, radius=8.1941, center=<-20.846 -0.0004 0>, rotation=<rot 90 0.0066 0.0066>}]
}
constraints [
 {bone1Name=dyn_strap_01_l, bone2Name=dyn_strap_01_l_MSTR, swing1LimitDegrees=25, swing2LimitDegrees=25, twistLimitDegrees=5}
 {bone1Name=dyn_strap_01_r, bone2Name=dyn_strap_01_r_MSTR, swing1LimitDegrees=25, swing2LimitDegrees=25, twistLimitDegrees=5}
 {bone1Name=dyn_zipper_01, bone2Name=dyn_zipper_01_MSTR, swing1LimitDegrees=5, swing2LimitDegrees=45, twistLimitDegrees=5}
 {bone1Name=dyn_zipper_01_r, bone2Name=dyn_zipper_01_r_MSTR, swing1LimitDegrees=10, swing2LimitDegrees=45, twistLimitDegrees=2}
 {bone1Name=dyn_zipper_01_l, bone2Name=dyn_zipper_01_l_MSTR, swing1LimitDegrees=10, swing2LimitDegrees=45, twistLimitDegrees=2}
 {bone1Name=dyn_tube_bk_01_l, bone2Name=dyn_tube_bk_01_l_MSTR, swing1LimitDegrees=25, swing2LimitDegrees=25, twistLimitDegrees=5}
 {bone1Name=dyn_tube_bk_02_l, bone2Name=dyn_tube_bk_01_l, swing1LimitDegrees=25, swing2LimitDegrees=25, twistLimitDegrees=5}
 {bone1Name=dyn_tube_bk_03_l, bone2Name=dyn_tube_bk_02_l, swing1LimitDegrees=25, swing2LimitDegrees=25, twistLimitDegrees=5}
 {bone1Name=dyn_tube_fr_01_l, bone2Name=dyn_tube_fr_01_l_MSTR, swing1LimitDegrees=5, swing2LimitDegrees=90, twistLimitDegrees=5}
 {bone1Name=dyn_tube_fr_02_l, bone2Name=dyn_tube_fr_01_l, swing1LimitDegrees=25, swing2LimitDegrees=25, twistLimitDegrees=5}
 {
  bone1Name=dyn_tube_fr_02_l
  bone2Name=dyn_tube_bk_03_l
  swing1LimitDegrees=45
  swing1Motion=Free
  swing2LimitDegrees=45
  swing2Motion=Free
  twistLimitDegrees=45
  twistMotion=Free
 }
 {
  bone1Name=dyn_tube_fr_01_l_MSTR
  bone2Name=thigh_l
  swing1LimitDegrees=45
  swing1Motion=Free
  swing2LimitDegrees=45
  swing2Motion=Free
  twistLimitDegrees=45
  twistMotion=Free
 }
 {
  bone1Name=dyn_strap_01_l_MSTR
  bone2Name=pelvis
  swing1LimitDegrees=45
  swing1Motion=Free
  swing2LimitDegrees=45
  swing2Motion=Free
  twistLimitDegrees=45
  twistMotion=Free
 }
 {
  bone1Name=dyn_strap_01_r_MSTR
  bone2Name=pelvis
  swing1LimitDegrees=45
  swing1Motion=Free
  swing2LimitDegrees=45
  swing2Motion=Free
  twistLimitDegrees=45
  twistMotion=Free
 }
 {
  bone1Name=dyn_zipper_01_MSTR
  bone2Name=pelvis
  swing1LimitDegrees=45
  swing1Motion=Free
  swing2LimitDegrees=45
  swing2Motion=Free
  twistLimitDegrees=45
  twistMotion=Free
 }
 {
  bone1Name=dyn_zipper_01_l_MSTR
  bone2Name=pelvis
  swing1LimitDegrees=45
  swing1Motion=Free
  swing2LimitDegrees=45
  swing2Motion=Free
  twistLimitDegrees=45
  twistMotion=Free
 }
 {
  bone1Name=dyn_zipper_01_r_MSTR
  bone2Name=pelvis
  swing1LimitDegrees=45
  swing1Motion=Free
  swing2LimitDegrees=45
  swing2Motion=Free
  twistLimitDegrees=45
  twistMotion=Free
 }
 {
  bone1Name=thigh_l
  bone2Name=pelvis
  swing1LimitDegrees=45
  swing1Motion=Free
  swing2LimitDegrees=45
  swing2Motion=Free
  twistLimitDegrees=45
  twistMotion=Free
 }
 {
  bone1Name=dyn_tube_bk_01_l_MSTR
  bone2Name=pelvis
  swing1LimitDegrees=45
  swing1Motion=Free
  swing2LimitDegrees=45
  swing2Motion=Free
  twistLimitDegrees=45
  twistMotion=Free
 }
 {
  bone1Name=dyn_tubeupper_bk_01_l_MSTR
  bone2Name=pelvis
  swing1LimitDegrees=45
  swing1Motion=Free
  swing2LimitDegrees=45
  swing2Motion=Free
  twistLimitDegrees=45
  twistMotion=Free
 }
 {
  bone1Name=dyn_buckle_01_r_MSTR
  bone2Name=pelvis
  swing1LimitDegrees=45
  swing1Motion=Free
  swing2LimitDegrees=45
  swing2Motion=Free
  twistLimitDegrees=45
  twistMotion=Free
 }
 {
  bone1Name=dyn_buckle_01_l_MSTR
  bone2Name=pelvis
  swing1LimitDegrees=45
  swing1Motion=Free
  swing2LimitDegrees=45
  swing2Motion=Free
  twistLimitDegrees=45
  twistMotion=Free
 }
 {bone1Name=dyn_buckle_01_l, bone2Name=dyn_buckle_01_l_MSTR, swing1LimitDegrees=5, swing2LimitDegrees=45, twistLimitDegrees=5}
 {bone1Name=dyn_buckle_01_r, bone2Name=dyn_buckle_01_r_MSTR, swing1LimitDegrees=5, swing2LimitDegrees=45, twistLimitDegrees=5}
 {bone1Name=dyn_tubeupper_fr_02_l, bone2Name=dyn_tubeupper_fr_01_l, swing1LimitDegrees=65, swing2LimitDegrees=65, twistLimitDegrees=5}
 {bone1Name=dyn_tubeupper_bk_01_l, bone2Name=dyn_tubeupper_bk_01_l_MSTR, swing1LimitDegrees=65, swing2LimitDegrees=65, twistLimitDegrees=5}
 {bone1Name=dyn_tubeupper_bk_02_l, bone2Name=dyn_tubeupper_bk_01_l, swing1LimitDegrees=65, swing2LimitDegrees=65, twistLimitDegrees=5}
 {
  bone1Name=dyn_tubeupper_bk_02_l
  bone2Name=dyn_tubeupper_fr_02_l
  swing1LimitDegrees=45
  swing1Motion=Free
  swing2LimitDegrees=45
  swing2Motion=Free
  twistLimitDegrees=45
  twistMotion=Free
 }
 {bone1Name=dyn_tubeupper_fr_01_l, bone2Name=dyn_tubeupper_fr_01_l_MSTR, swing1LimitDegrees=45, swing2LimitDegrees=45, twistLimitDegrees=5}
 {
  bone1Name=dyn_tubeupper_fr_01_l_MSTR
  bone2Name=pelvis
  swing1LimitDegrees=45
  swing1Motion=Free
  swing2LimitDegrees=45
  swing2Motion=Free
  twistLimitDegrees=45
  twistMotion=Free
 }
]
