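# Post-process style AnimBlueprint for the M_MED_York body cosmetic, copying the parent mesh pose and animating its accessory bones.
# Weighted multi-bone constraints place zippers, straps, buckles and gear, while crouch state swaps two gear look-ats for additive offsets.
# EventGraph-computed shoulder-to-neck distance and calf twist drive neck and knee correctives.
# Two RigidBody simulations finish the pose, with gravity overridden by scaled skydiving force while skydiving.
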AnimBlueprint /BRCosmetics/Characters/Player/Male/Medium/Bodies/M_MED_York/Meshes/M_MED_York_AnimBP
target_skeleton=&/M_MED_York_Skeleton
graphs [
 AnimGraph {nodes [
   n1 = "Constraint - dyn_zipper_01_l_MSTR (PPPPPPPP)"!(
    ComponentPose=Local To Component(LocalPose=Copy Pose From Mesh() {bUseAttachedParent=true})
    Alpha=1
    ConstraintWeights_0=0.0012
    ConstraintWeights_1:float=0.0158
    ConstraintWeights_2:float=0.1493
    ConstraintWeights_3:float=0.2734
    ConstraintWeights_4:float=0.4309
    ConstraintWeights_5:float=0.1253
    ConstraintWeights_6:float=0
    ConstraintWeights_7:float=0.0041
   ) -> (Pose) {
    alphaBoolBlend {blendOption=Linear}
    boneToModify {boneName=dyn_zipper_01_l_MSTR}
    constraintSetup [
     {targetBone {boneName=clavicle_l}}
     {targetBone {boneName=spine_01}}
     {targetBone {boneName=spine_02}}
     {targetBone {boneName=spine_03}}
     {targetBone {boneName=spine_04}}
     {targetBone {boneName=spine_05}}
     {targetBone {boneName=neck_02}}
     {targetBone {boneName=pelvis}}
    ]
    constraintWeights [1,1,1,1,1,1,1,1]
   }
   n2 = "Constraint - dyn_zipper_01_r_MSTR (PPPPPPPP)"!(
    ComponentPose=@n1
    Alpha=1
    ConstraintWeights_0=0.0157
    ConstraintWeights_1:float=0.1484
    ConstraintWeights_2:float=0.2728
    ConstraintWeights_3:float=0.4317
    ConstraintWeights_4:float=0.126
    ConstraintWeights_5:float=0
    ConstraintWeights_6:float=0.0013
    ConstraintWeights_7:float=0.004
   ) -> (Pose) {
    alphaBoolBlend {blendOption=Linear}
    boneToModify {boneName=dyn_zipper_01_r_MSTR}
    constraintSetup [
     {targetBone {boneName=spine_01}}
     {targetBone {boneName=spine_02}}
     {targetBone {boneName=spine_03}}
     {targetBone {boneName=spine_04}}
     {targetBone {boneName=spine_05}}
     {targetBone {boneName=neck_02}}
     {targetBone {boneName=clavicle_r}}
     {targetBone {boneName=pelvis}}
    ]
    constraintWeights [1,1,1,1,1,1,1,1]
   }
   n3 = "Constraint - dyn_strap_01_r_MSTR (PPPPPPPPPP)"!(
    ComponentPose=@n2
    Alpha=1
    ConstraintWeights_0=0.0003
    ConstraintWeights_1:float=0.1649
    ConstraintWeights_2:float=0.0002
    ConstraintWeights_3:float=0.0039
    ConstraintWeights_4:float=0.1538
    ConstraintWeights_5:float=0.5339
    ConstraintWeights_6:float=0.0651
    ConstraintWeights_7:float=0.0779
    ConstraintWeights_8:float=0
    ConstraintWeights_9:float=0
   ) -> (Pose) {
    alphaBoolBlend {blendOption=Linear}
    boneToModify {boneName=dyn_strap_01_r_MSTR}
    constraintSetup [
     {targetBone {boneName=upperarm_r}}
     {targetBone {boneName=upperarm_twist_01_r}}
     {targetBone {boneName=upperarm_twist_02_r}}
     {targetBone {boneName=spine_02}}
     {targetBone {boneName=spine_03}}
     {targetBone {boneName=spine_04}}
     {targetBone {boneName=spine_05}}
     {targetBone {boneName=clavicle_r}}
     {targetBone {boneName=neck_02}}
     {targetBone {boneName=neck_01}}
    ]
    constraintWeights [1,1,1,1,1,1,1,1,1,1]
   }
   n4 = "Constraint - dyn_strap_01_l_MSTR (PPPPPPPPPP)"!(
    ComponentPose=@n3
    Alpha=1
    ConstraintWeights_0=0.0779
    ConstraintWeights_1:float=0.0002
    ConstraintWeights_2:float=0.0039
    ConstraintWeights_3:float=0.1538
    ConstraintWeights_4:float=0.5339
    ConstraintWeights_5:float=0.0651
    ConstraintWeights_6:float=0.1649
    ConstraintWeights_7:float=0
    ConstraintWeights_8:float=0
    ConstraintWeights_9:float=0.0003
   ) -> (Pose) {
    alphaBoolBlend {blendOption=Linear}
    boneToModify {boneName=dyn_strap_01_l_MSTR}
    constraintSetup [
     {targetBone {boneName=clavicle_l}}
     {targetBone {boneName=upperarm_twist_02_l}}
     {targetBone {boneName=spine_02}}
     {targetBone {boneName=spine_03}}
     {targetBone {boneName=spine_04}}
     {targetBone {boneName=spine_05}}
     {targetBone {boneName=upperarm_twist_01_l}}
     {targetBone {boneName=neck_02}}
     {targetBone {boneName=neck_01}}
     {targetBone {boneName=upperarm_l}}
    ]
    constraintWeights [1,1,1,1,1,1,1,1,1,1]
   }
   n5 = "Constraint - dyn_zipper_01_MSTR (PPPPP)"!(
    ComponentPose=@n4
    Alpha=1
    ConstraintWeights_0=0.0002
    ConstraintWeights_1:float=0.0256
    ConstraintWeights_2:float=0.1287
    ConstraintWeights_3:float=0.4809
    ConstraintWeights_4:float=0.3646
   ) -> (Pose) {
    alphaBoolBlend {blendOption=Linear}
    boneToModify {boneName=dyn_zipper_01_MSTR}
    constraintSetup [
     {targetBone {boneName=neck_02}}
     {targetBone {boneName=spine_02}}
     {targetBone {boneName=spine_03}}
     {targetBone {boneName=spine_04}}
     {targetBone {boneName=spine_05}}
    ]
    constraintWeights [1,1,1,1,1]
   }
   n6 = "Constraint - dyn_mech_fr_01_l (PPPPPPP)"!(
    ComponentPose=@n5
    Alpha=1
    ConstraintWeights_0=0
    ConstraintWeights_1:float=0.3377
    ConstraintWeights_2:float=0.0497
    ConstraintWeights_3:float=0.018
    ConstraintWeights_4:float=0
    ConstraintWeights_5:float=0.582
    ConstraintWeights_6:float=0.0126
   ) -> (Pose) {
    alphaBoolBlend {blendOption=Linear}
    boneToModify {boneName=dyn_mech_fr_01_l}
    constraintSetup [
     {targetBone {boneName=thigh_twist_01_r}}
     {targetBone {boneName=spine_01}}
     {targetBone {boneName=spine_02}}
     {targetBone {boneName=spine_03}}
     {targetBone {boneName=spine_04}}
     {targetBone {boneName=pelvis}}
     {targetBone {boneName=thigh_twist_01_l}}
    ]
    constraintWeights [1,1,1,1,1,1,1]
   }
   n7 = "Constraint - dyn_tube_fr_01_l_MSTR (PPPP)"!(
    ComponentPose=@n6
    Alpha=1
    ConstraintWeights_0=0
    ConstraintWeights_1:float=0.9155
    ConstraintWeights_2:float=0.0087
    ConstraintWeights_3:float=0.0758
   ) -> (Pose) {
    alphaBoolBlend {blendOption=Linear}
    boneToModify {boneName=dyn_tube_fr_01_l_MSTR}
    constraintSetup [{targetBone {boneName=spine_01}}, {targetBone {boneName=thigh_twist_01_l}}, {targetBone {boneName=thigh_l}}, {targetBone {boneName=pelvis}}]
    constraintWeights [1,1,1,1]
   }
   n8 = "Constraint - dyn_buckle_01_l_MSTR (PPPPPPP)"!(
    ComponentPose=@n7
    Alpha=1
    ConstraintWeights_0=0
    ConstraintWeights_1:float=0.18
    ConstraintWeights_2:float=0.004
    ConstraintWeights_3:float=0.0009
    ConstraintWeights_4:float=0
    ConstraintWeights_5:float=0.685
    ConstraintWeights_6:float=0.1301
   ) -> (Pose) {
    alphaBoolBlend {blendOption=Linear}
    boneToModify {boneName=dyn_buckle_01_l_MSTR}
    constraintSetup [
     {targetBone {boneName=thigh_twist_01_r}}
     {targetBone {boneName=spine_01}}
     {targetBone {boneName=spine_02}}
     {targetBone {boneName=spine_03}}
     {targetBone {boneName=spine_04}}
     {targetBone {boneName=pelvis}}
     {targetBone {boneName=thigh_twist_01_l}}
    ]
    constraintWeights [1,1,1,1,1,1,1]
   }
   n9 = "Constraint - dyn_buckle_01_r_MSTR (PPPPP)"!(
    ComponentPose=@n8
    Alpha=1
    ConstraintWeights_0=0.1228
    ConstraintWeights_1:float=0.0016
    ConstraintWeights_2:float=0.0006
    ConstraintWeights_3:float=0.1092
    ConstraintWeights_4:float=0.7658
   ) -> (Pose) {
    alphaBoolBlend {blendOption=Linear}
    boneToModify {boneName=dyn_buckle_01_r_MSTR}
    constraintSetup [
     {targetBone {boneName=spine_01}}
     {targetBone {boneName=spine_02}}
     {targetBone {boneName=spine_03}}
     {targetBone {boneName=thigh_twist_01_r}}
     {targetBone {boneName=pelvis}}
    ]
    constraintWeights [1,1,1,1,1]
   }
   n10 = "Constraint - dyn_mech_fr_02_l_aim (PPPP)"!(
    ComponentPose=@n9
    Alpha=1
    ConstraintWeights_0=0
    ConstraintWeights_1:float=0.9155
    ConstraintWeights_2:float=0.0087
    ConstraintWeights_3:float=0.0758
   ) -> (Pose) {
    alphaBoolBlend {blendOption=Linear}
    boneToModify {boneName=dyn_mech_fr_02_l_aim}
    constraintSetup [{targetBone {boneName=spine_01}}, {targetBone {boneName=thigh_twist_01_l}}, {targetBone {boneName=thigh_l}}, {targetBone {boneName=pelvis}}]
    constraintWeights [1,1,1,1]
   }
   n11 = Get isNotCrouching() -> (isNotCrouching:double)
   n12 = Look At - Bone: dyn_mech_fr_01_l(ComponentPose=@n10, Alpha=@n11) {
    alphaBoolBlend {blendOption=Linear}
    boneToModify {boneName=dyn_mech_fr_01_l}
    lookAtLocation {x=0}
    lookAtTarget {boneReference {boneName=dyn_mech_fr_02_l_aim}}
    lookAt_Axis {axis=<1 0>}
    lookUp_Axis {axis {x=0, z=1}}
   }
   n13 = "Transform (Modify) Bone - Bone: dyn_mech_fr_01_l"(ComponentPose=@n12, Alpha=@n11, Translation="0,0,0", Rotation="5.000,0.000,0.000", Scale="1,1,1") {
    rotationMode=BMM_Additive
    rotationSpace=BCS_BoneSpace
    scaleMode=BMM_Replace
    scaleSpace=BCS_BoneSpace
    translationMode=BMM_Additive
    translationSpace=BCS_BoneSpace
    alphaBoolBlend {blendOption=Linear}
    alphaScaleBiasClamp {interpSpeedDecreasing=15, interpSpeedIncreasing=15}
    boneToModify {boneName=dyn_mech_fr_01_l}
    rotation {pitch=5}
   }
   n14 = "Transform (Modify) Bone - Bone: dyn_mech_fr_01_l"(
    ComponentPose=@n13
    Alpha=Get isCrouchingAlpha() -> (isCrouchingAlpha:double)
    Translation="-0.246,-3.914,-1.286"
    Rotation="8.585,-5.895,-30.180"
    Scale="1,1,1"
   ) {
    rotationMode=BMM_Additive
    rotationSpace=BCS_BoneSpace
    scaleMode=BMM_Replace
    scaleSpace=BCS_BoneSpace
    translationMode=BMM_Additive
    translationSpace=BCS_BoneSpace
    alphaBoolBlend {blendOption=Linear}
    alphaScaleBiasClamp {interpSpeedDecreasing=15, interpSpeedIncreasing=15}
    boneToModify {boneName=dyn_mech_fr_01_l}
    rotation=<rot 8.585 -5.8951 -30.1801>
    translation=<-0.246 -3.914 -1.286>
   }
   n15 = "Constraint - dyn_walkietalkie_01_r (PPPPPP)"!(
    ComponentPose=@n14
    Alpha=0
    ConstraintWeights_0=0.0317
    ConstraintWeights_1:float=0.32
    ConstraintWeights_2:float=0.035
    ConstraintWeights_3:float=0.0129
    ConstraintWeights_4:float=0
    ConstraintWeights_5:float=0.6004
   ) -> (Pose) {
    alphaBoolBlend {blendOption=Linear}
    boneToModify {boneName=dyn_walkietalkie_01_r}
    constraintSetup [
     {targetBone {boneName=thigh_twist_01_r}}
     {targetBone {boneName=spine_01}}
     {targetBone {boneName=spine_02}}
     {targetBone {boneName=spine_03}}
     {targetBone {boneName=spine_04}}
     {targetBone {boneName=pelvis}}
    ]
    constraintWeights [1,1,1,1,1,1]
   }
   n16 = "Constraint - dyn_walkietalkie_01_r_aim (PPPPPPPP)"!(
    ComponentPose=@n15
    Alpha=1
    ConstraintWeights_0=0.0095
    ConstraintWeights_1:float=0.452
    ConstraintWeights_2:float=0.5374
    ConstraintWeights_3:float=0
    ConstraintWeights_4:float=0
    ConstraintWeights_5:float=0
    ConstraintWeights_6:float=0.001
    ConstraintWeights_7:float=0.0001
   ) -> (Pose) {
    alphaBoolBlend {blendOption=Linear}
    boneToModify {boneName=dyn_walkietalkie_01_r_aim}
    constraintSetup [
     {targetBone {boneName=calf_r}}
     {targetBone {boneName=thigh_r}}
     {targetBone {boneName=thigh_twist_01_r}}
     {targetBone {boneName=spine_01}}
     {targetBone {boneName=thigh_twist_01_l}}
     {targetBone {boneName=thigh_l}}
     {targetBone {boneName=pelvis}}
     {targetBone {boneName=calf_twist_02_r}}
    ]
    constraintWeights [1,1,1,1,1,1,1,1]
   }
   n17 = Look At - Bone: dyn_walkietalkie_01_r(ComponentPose=@n16, Alpha=Get isNotCrouching()) {
    alphaBoolBlend {blendOption=Linear}
    alphaScaleBiasClamp {bInterpResult=true}
    boneToModify {boneName=dyn_walkietalkie_01_r}
    lookAtLocation {x=0}
    lookAtTarget {boneReference {boneName=dyn_walkietalkie_01_r_aim}}
    lookAt_Axis {axis=<1 0>}
    lookUp_Axis {axis {x=0, z=1}}
   }
   n18 = "Transform (Modify) Bone - Bone: dyn_walkietalkie_01_r"(ComponentPose=@n17, Alpha=0.5, Translation="1.411,-0.000,0.000", Rotation="4.982,10.010,0.436", Scale="1,1,1") {
    rotationMode=BMM_Additive
    rotationSpace=BCS_BoneSpace
    translationMode=BMM_Additive
    translationSpace=BCS_BoneSpace
    alphaBoolBlend {blendOption=Linear}
    boneToModify {boneName=dyn_walkietalkie_01_r}
    rotation=<rot 4.982 10.0103 0.436>
    translation {x=1.411}
   }
   n19 = "Transform (Modify) Bone - Bone: dyn_walkietalkie_01_r"(ComponentPose=@n18, Alpha=Get isCrouchingAlpha(), Translation="0.412,1.786,-0.121", Rotation="-1.924,28.871,11.543", Scale="1,1,1") {
    rotationMode=BMM_Additive
    rotationSpace=BCS_BoneSpace
    translationMode=BMM_Additive
    translationSpace=BCS_BoneSpace
    alphaBoolBlend {blendOption=Linear}
    alphaScaleBiasClamp {bInterpResult=true}
    boneToModify {boneName=dyn_walkietalkie_01_r}
    rotation=<rot -1.924 28.871 11.5432>
    translation=<0.412 1.786 -0.121>
   }
   n20 = "Transform (Modify) Bone - Bone: neck_01_adjust_JNT"(ComponentPose=@n19, Alpha=Get neckDistance_l() -> (neckDistance_l:double), Translation="0,0,0", Rotation="0,0,0", Scale="1.000,1.000,0.3") {
    rotationMode=BMM_Additive
    rotationSpace=BCS_BoneSpace
    scaleMode=BMM_Additive
    scaleSpace=BCS_BoneSpace
    translationMode=BMM_Additive
    translationSpace=BCS_BoneSpace
    alphaBoolBlend {blendOption=Linear}
    boneToModify {boneName=neck_01_adjust_JNT}
    scale {z=0.3}
   }
   n21 = "Transform (Modify) Bone - Bone: knee_01_r"(ComponentPose=@n20, Alpha=Get kneeAlpha_r() -> (kneeAlpha_r:double), Translation="0,0,0", Rotation="0.000,-50,0.000", Scale="1,1,1") {
    rotationMode=BMM_Additive
    rotationSpace=BCS_BoneSpace
    scaleMode=BMM_Additive
    scaleSpace=BCS_BoneSpace
    translationMode=BMM_Additive
    translationSpace=BCS_BoneSpace
    alphaBoolBlend {blendOption=Linear}
    alphaScaleBiasClamp {interpSpeedDecreasing=5, interpSpeedIncreasing=5}
    boneToModify {boneName=knee_01_r}
    rotation {yaw=-50}
   }
   n22 = "Transform (Modify) Bone - Bone: knee_01_l"(ComponentPose=@n21, Alpha=Get kneeAlpha_l() -> (kneeAlpha_l:double), Translation="0,0,0", Rotation="0.000,25.000,0.000", Scale="1,1,1") {
    rotationMode=BMM_Additive
    rotationSpace=BCS_BoneSpace
    scaleMode=BMM_Additive
    scaleSpace=BCS_BoneSpace
    translationMode=BMM_Additive
    translationSpace=BCS_BoneSpace
    alphaBoolBlend {blendOption=Linear}
    boneToModify {boneName=knee_01_l}
    rotation {yaw=24.9998}
   }
   n23 = Get windVector() -> (windVector:Vector)
   n24 = RigidBody!(ComponentPose=@n22, Alpha, OverrideWorldGravity:Vector=@n23, ExternalForce) -> (Pose) {
    bClampLinearTranslationLimitToRefPose=true
    bOverrideWorldGravity=true
    overridePhysicsAsset=&/M_MED_York_Physics
    simulationSpace=BaseBoneSpace
    alphaBoolBlend {blendOption=Linear}
    baseBoneRef {boneName=pelvis}
    componentAppliedLinearAccClamp=<3000 3000 3000>
    componentLinearAccScale=<0.4 0.4 0.4>
    componentLinearVelScale=<0.4 0.4 0.4>
    overrideWorldGravity {z=-980}
   }
   n25 = RigidBody(ComponentPose=@n24, OverrideWorldGravity=@n23) {
    bClampLinearTranslationLimitToRefPose=true
    bOverrideWorldGravity=true
    overridePhysicsAsset=&/M_MED_York_dyn_zippers
    simulationSpace=BaseBoneSpace
    alphaBoolBlend {blendOption=Linear}
    baseBoneRef {boneName=spine_04}
    componentAppliedLinearAccClamp=<3000 3000 3000>
    componentLinearAccScale=<0.45 0.45 0.45>
    componentLinearVelScale=<0.4 0.4 0.4>
    overrideWorldGravity {z=-980}
   }
   Output Pose(Result=Component To Local(ComponentPose=@n25)) {layerGroup=None}
  ]}
 EventGraph {nodes [
   Event BlueprintUpdateAnimation()
   > n1 = Sequence!(execute) -> (then_0, then_1, then_2:Exec, then_3:Exec, then_4:Exec, then_5:Exec)
   n2 = Break Fort Anim Input Skydiving External Force(FortAnimInput_SkydivingExternalForce=Get SkydivingExternalForceInput()) -> (
    bUseSkydivingVectorForce=~
    SkydivingVectorForce=~
    HeadToPelvisDirection=~
    FloatingMultiplier=~
    FloatingAdditive=~
    DivingMultiplier=~
    DivingAdditive=~
    ParachutingMultiplier=~
    ParachutingAdditive=~
   )
   > Set windVector(windVector:Vector=SelectVector(A=vector * float(A=@n2.SkydivingVectorForce, B=0.15), B="0, 0,-980", bPickA=Get bIsSkyDiving())) -> (Output_Get:Vector)
   n3 = Set isCrouchingAlpha(execute=@n1.then_1, isCrouchingAlpha:double=SelectFloat(A=1, B=0, bPickA=Get bIsCrouching())) -> (Output_Get:double)
   > Set isNotCrouching(isNotCrouching:double=float - float(B=@n3) {is=title K2Node_CallFunction}) -> (Output_Get:double)
   n4 = GetPartSkeletalMeshComponent()
   n5 = TransformFromBoneSpace!(
    execute=@n1.then_2
    self:Skinned Mesh Component Object Reference=@n4
    BoneName:Name=upperarm_r
    InPosition:Vector="0, 0, 0"
    InRotation:Rotator="0, 0, 0"
   ) -> (then, OutPosition:Vector, OutRotation:Rotator)
   > n6 = TransformFromBoneSpace(self=@n4, BoneName=neck_02, InPosition="0, 0, 0", InRotation="0, 0, 0")
   > Set neckDistance_l(neckDistance_l:double=MapRangeClamped(Value="Distance (Vector)"!(V1:Vector=@n5, V2:Vector=@n6) -> (ReturnValue:double), InRangeA=18.3, InRangeB=17.8)) -> (Output_Get:double)
   Set kneeAlpha_r(
    execute=@n1.then_3
    kneeAlpha_r:double=GetAlphaFromReferenceTransformDelta(
     SkeletalMeshComponent=@n4
     BoneName=calf_r
     BaseName=thigh_r
     DeltaComponentType=QuaternionTwist
     InputRange=(X=-100,Y=-130)
     TwistAxis="0,0,1"
    )
   ) -> (Output_Get:double)
   > Set kneeAlpha_l(kneeAlpha_l:double=GetAlphaFromReferenceTransformDelta(
     SkeletalMeshComponent=@n4
     BoneName=calf_l
     BaseName=thigh_l
     DeltaComponentType=QuaternionTwist
     InputRange=(X=-100,Y=-130)
     TwistAxis="0,0,1"
    )) -> (Output_Get:double)
   TryGetPawnOwner()
  ]}
]
variables [
 isCrouchingAlpha {type=number}
 isNotCrouching {type=number}
 kneeAlpha_l {type=number}
 kneeAlpha_r {type=number}
 neckDistance_l {type=number}
 windVector {type=Vector}
]
